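# Animation Blueprint for the female RoseScour body cosmetic, adding secondary motion to clothing and accessories.
# It copies the attached parent's pose, layers in HeadMesh facial-root, jaw, chin and ear bones, then runs a Control Rig.
# Chained root-relative Anim Dynamics nodes simulate jacket, collar, knots, ropes and straps.
# The event graph caches head and tail meshes, hides the tail under cape backblings, and sets front-end or in-game sim-space variables.
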AnimBlueprint /BRCosmetics/Characters/Player/Female/Medium/Bodies/F_MED_RoseScour/Meshes/F_MED_RoseScour_AnimBP
target_skeleton=&/F_MED_RoseScour_Skeleton
graphs [
 AnimGraph {nodes [
   n1 = Layered blend per bone(
    BasePose=Copy Pose From Mesh() {bUseAttachedParent=true}
    BlendPoses_0=Copy Pose From Mesh(SourceMeshComponent=Get HeadMesh() -> (HeadMesh:Skeletal Mesh Component Object Reference))
   ) {layerSetup [{branchFilters [
       {boneName=FACIAL_C_FacialRoot}
       {boneName=FACIAL_C_Jaw}
       {boneName=FACIAL_C_Jawline}
       {boneName=FACIAL_C_UnderChin}
       {boneName=FACIAL_L_Jawline}
       {boneName=FACIAL_L_UnderChin}
       {boneName=FACIAL_R_Jawline}
       {boneName=FACIAL_R_UnderChin}
       {boneName=FACIAL_L_Ear}
       {boneName=FACIAL_R_Ear}
      ]}]}
   n2 = Control Rig(Source=@n1) {controlRigAssetReference {blueprintRigClass=&/F_MED_RoseScour_CtrlRig.F_MED_RoseScour_CtrlRig_C}}
   n3 = Anim Dynamics - Bone: dyn_jacket_l_A_01!(
    ComponentPose=Local To Component(LocalPose=@n2)
    Alpha=1
    SimSpaceSettings:Anim Phys Sim Space Settings Structure=(SimSpaceAngularAlpha=0,MaxAngularVelocity=10000,MaxAngularAcceleration=10000,ExternalAngularVelocity=(X=0,Y=0,Z=0))
    ExternalForce="0,0,0"
   ) -> (Pose) {
    angularDampingOverride=1
    angularSpringConstant=6000
    bAngularSpring=true
    bOverrideAngularDamping=true
    bUseGravityOverride=true
    simulationSpace=RootRelative
    alphaBoolBlend {blendOption=Linear}
    boundBone {boneName=dyn_jacket_l_A_01}
    componentAppliedLinearAccClamp=<3000 3000 3000>
    componentLinearAccScale=<0.2 0.6 0.2>
    componentLinearVelScale=<0.4 1 0.4>
    gravityOverride {z=-300}
    physicsBodyDefinitions [{
      boundBone {boneName=dyn_jacket_l_A_01}
      boxExtents=<25 12.5 30>
      constraintSetup {angularLimitsMax=<0 5 2.5>, angularLimitsMin=<0 -70 -2.5>, angularTarget=<1 0 0>}
      localJointOffset=<15 0 0>
     }]
    relativeSpaceBone {boneName=head}
   }
   n4 = Anim Dynamics - Bone: dyn_jacket_l_B_01!(
    ComponentPose=@n3
    Alpha=1
    SimSpaceSettings:Anim Phys Sim Space Settings Structure=(SimSpaceAngularAlpha=0,MaxAngularVelocity=10000,MaxAngularAcceleration=10000,ExternalAngularVelocity=(X=0,Y=0,Z=0))
    ExternalForce="0,0,0"
   ) -> (Pose) {
    angularDampingOverride=1
    angularSpringConstant=6000
    bAngularSpring=true
    bOverrideAngularDamping=true
    bUseGravityOverride=true
    simulationSpace=RootRelative
    alphaBoolBlend {blendOption=Linear}
    boundBone {boneName=dyn_jacket_l_B_01}
    componentAppliedLinearAccClamp=<3000 3000 3000>
    componentLinearAccScale=<0.2 0.6 0.1>
    componentLinearVelScale=<0.4 1 0.2>
    gravityOverride {z=-300}
    physicsBodyDefinitions [{
      boundBone {boneName=dyn_jacket_l_B_01}
      boxExtents=<25 10 30>
      constraintSetup {angularLimitsMax=<0 10 2.5>, angularLimitsMin=<0 -50 -2.5>, angularTarget=<1 0 0>}
      localJointOffset=<15 0 0>
     }]
    relativeSpaceBone {boneName=head}
   }
   n5 = Anim Dynamics - Bone: dyn_jacket_l_C_01!(
    ComponentPose=@n4
    Alpha=1
    SimSpaceSettings:Anim Phys Sim Space Settings Structure=(SimSpaceAngularAlpha=0,MaxAngularVelocity=10000,MaxAngularAcceleration=10000,ExternalAngularVelocity=(X=0,Y=0,Z=0))
    ExternalForce="0,0,0"
   ) -> (Pose) {
    angularDampingOverride=1
    angularSpringConstant=6000
    bAngularSpring=true
    bOverrideAngularDamping=true
    bUseGravityOverride=true
    simulationSpace=RootRelative
    alphaBoolBlend {blendOption=Linear}
    boundBone {boneName=dyn_jacket_l_C_01}
    componentAppliedLinearAccClamp=<3000 3000 3000>
    componentLinearAccScale=<0.2 0.6 0.1>
    componentLinearVelScale=<0.4 1 0.2>
    gravityOverride {z=-300}
    physicsBodyDefinitions [{
      boundBone {boneName=dyn_jacket_l_C_01}
      boxExtents=<25 10 30>
      constraintSetup {angularLimitsMax=<0 10 2.5>, angularLimitsMin=<0 -50 -2.5>, angularTarget=<1 0 0>}
      localJointOffset=<15 0 0>
     }]
    relativeSpaceBone {boneName=head}
   }
   n6 = Anim Dynamics - Bone: dyn_jacket_bk_01!(
    ComponentPose=@n5
    Alpha=1
    SimSpaceSettings:Anim Phys Sim Space Settings Structure=(SimSpaceAngularAlpha=0,MaxAngularVelocity=10000,MaxAngularAcceleration=10000,ExternalAngularVelocity=(X=0,Y=0,Z=0))
    ExternalForce="0,0,0"
   ) -> (Pose) {
    angularDampingOverride=1
    angularSpringConstant=6000
    bAngularSpring=true
    bOverrideAngularDamping=true
    bUseGravityOverride=true
    simulationSpace=RootRelative
    alphaBoolBlend {blendOption=Linear}
    boundBone {boneName=dyn_jacket_bk_01}
    componentAppliedLinearAccClamp=<3000 3000 3000>
    componentLinearAccScale=<0.2 0.6 0.1>
    componentLinearVelScale=<0.4 1 0.2>
    gravityOverride {z=-300}
    physicsBodyDefinitions [{
      boundBone {boneName=dyn_jacket_bk_01}
      boxExtents=<25 10 30>
      constraintSetup {angularLimitsMax=<0 10 2.5>, angularLimitsMin=<0 -50 -2.5>, angularTarget=<1 0 0>}
      localJointOffset=<15 0 0>
     }]
    relativeSpaceBone {boneName=head}
   }
   n7 = Anim Dynamics - Bone: dyn_jacket_r_A_01!(
    ComponentPose=@n6
    Alpha=1
    SimSpaceSettings:Anim Phys Sim Space Settings Structure=(SimSpaceAngularAlpha=0,MaxAngularVelocity=10000,MaxAngularAcceleration=10000,ExternalAngularVelocity=(X=0,Y=0,Z=0))
    ExternalForce="0,0,0"
   ) -> (Pose) {
    angularDampingOverride=1
    angularSpringConstant=6000
    bAngularSpring=true
    bOverrideAngularDamping=true
    bUseGravityOverride=true
    simulationSpace=RootRelative
    alphaBoolBlend {blendOption=Linear}
    boundBone {boneName=dyn_jacket_r_A_01}
    componentAppliedLinearAccClamp=<3000 3000 3000>
    componentLinearAccScale=<0.2 0.6 0.2>
    componentLinearVelScale=<0.4 1 0.4>
    gravityOverride {z=-300}
    physicsBodyDefinitions [{
      boundBone {boneName=dyn_jacket_r_A_01}
      boxExtents=<25 10 30>
      constraintSetup {angularLimitsMax=<0 5 2.5>, angularLimitsMin=<0 -70 -2.5>, angularTarget=<1 0 0>}
      localJointOffset=<15 0 0>
     }]
    relativeSpaceBone {boneName=head}
   }
   n8 = Anim Dynamics - Bone: dyn_jacket_r_B_01!(
    ComponentPose=@n7
    Alpha=1
    SimSpaceSettings:Anim Phys Sim Space Settings Structure=(SimSpaceAngularAlpha=0,MaxAngularVelocity=10000,MaxAngularAcceleration=10000,ExternalAngularVelocity=(X=0,Y=0,Z=0))
    ExternalForce="0,0,0"
   ) -> (Pose) {
    angularDampingOverride=1
    angularSpringConstant=6000
    bAngularSpring=true
    bOverrideAngularDamping=true
    bUseGravityOverride=true
    simulationSpace=RootRelative
    alphaBoolBlend {blendOption=Linear}
    boundBone {boneName=dyn_jacket_r_B_01}
    componentAppliedLinearAccClamp=<3000 3000 3000>
    componentLinearAccScale=<0.2 0.6 0.1>
    componentLinearVelScale=<0.4 1 0.2>
    gravityOverride {z=-300}
    physicsBodyDefinitions [{
      boundBone {boneName=dyn_jacket_r_B_01}
      boxExtents=<25 10 30>
      constraintSetup {angularLimitsMax=<0 10 2.5>, angularLimitsMin=<0 -50 -2.5>, angularTarget=<1 0 0>}
      localJointOffset=<15 0 0>
     }]
    relativeSpaceBone {boneName=head}
   }
   n9 = Anim Dynamics - Bone: dyn_jacket_r_C_01!(
    ComponentPose=@n8
    Alpha=1
    SimSpaceSettings:Anim Phys Sim Space Settings Structure=(SimSpaceAngularAlpha=0,MaxAngularVelocity=10000,MaxAngularAcceleration=10000,ExternalAngularVelocity=(X=0,Y=0,Z=0))
    ExternalForce="0,0,0"
   ) -> (Pose) {
    angularDampingOverride=1
    angularSpringConstant=6000
    bAngularSpring=true
    bOverrideAngularDamping=true
    bUseGravityOverride=true
    simulationSpace=RootRelative
    alphaBoolBlend {blendOption=Linear}
    boundBone {boneName=dyn_jacket_r_C_01}
    componentAppliedLinearAccClamp=<3000 3000 3000>
    componentLinearAccScale=<0.2 0.6 0.1>
    componentLinearVelScale=<0.4 1 0.2>
    gravityOverride {z=-300}
    physicsBodyDefinitions [{
      boundBone {boneName=dyn_jacket_r_C_01}
      boxExtents=<25 10 30>
      constraintSetup {angularLimitsMax=<0 10 2.5>, angularLimitsMin=<0 -50 -2.5>, angularTarget=<1 0 0>}
      localJointOffset=<15 0 0>
     }]
    relativeSpaceBone {boneName=head}
   }
   n10 = Anim Dynamics - Bone: dyn_collar_l_01!(
    ComponentPose=@n9
    Alpha=1
    SimSpaceSettings:Anim Phys Sim Space Settings Structure=(SimSpaceAngularAlpha=0,MaxAngularVelocity=10000,MaxAngularAcceleration=10000,ExternalAngularVelocity=(X=0,Y=0,Z=0))
    ExternalForce="0,0,0"
   ) -> (Pose) {
    angularDampingOverride=1
    angularSpringConstant=7500
    bAngularSpring=true
    bOverrideAngularDamping=true
    bUseGravityOverride=true
    simulationSpace=RootRelative
    alphaBoolBlend {blendOption=Linear}
    boundBone {boneName=dyn_collar_l_01}
    componentAppliedLinearAccClamp=<3000 3000 3000>
    componentLinearAccScale=<0.4 0.4 0.4>
    componentLinearVelScale=<0.8 0.8 0.8>
    gravityOverride {z=980}
    physicsBodyDefinitions [{
      boundBone {boneName=dyn_collar_l_01}
      boxExtents=<25 20 20>
      constraintSetup {angularLimitsMax=<0 10 5>, angularLimitsMin=<0 -30 -5>, angularTarget=<1 0 0>}
      localJointOffset=<3.5 1 -1.25>
     }]
    relativeSpaceBone {boneName=head}
   }
   n11 = Anim Dynamics - Bone: dyn_collar_r_01!(
    ComponentPose=@n10
    Alpha=1
    SimSpaceSettings:Anim Phys Sim Space Settings Structure=(SimSpaceAngularAlpha=0,MaxAngularVelocity=10000,MaxAngularAcceleration=10000,ExternalAngularVelocity=(X=0,Y=0,Z=0))
    ExternalForce="0,0,0"
   ) -> (Pose) {
    angularDampingOverride=1
    angularSpringConstant=7500
    bAngularSpring=true
    bOverrideAngularDamping=true
    bUseGravityOverride=true
    simulationSpace=RootRelative
    alphaBoolBlend {blendOption=Linear}
    boundBone {boneName=dyn_collar_r_01}
    componentAppliedLinearAccClamp=<3000 3000 3000>
    componentLinearAccScale=<0.4 0.4 0.4>
    componentLinearVelScale=<0.8 0.8 0.8>
    gravityOverride {z=980}
    physicsBodyDefinitions [{
      boundBone {boneName=dyn_collar_r_01}
      boxExtents=<25 20 20>
      constraintSetup {angularLimitsMax=<0 10 5>, angularLimitsMin=<0 -30 -5>, angularTarget=<1 0 0>}
      localJointOffset=<3.5 -1 -2>
     }]
    relativeSpaceBone {boneName=head}
   }
   n12 = Anim Dynamics - Bone: dyn_knots_A_01!(
    ComponentPose=@n11
    Alpha=1
    SimSpaceSettings:Anim Phys Sim Space Settings Structure=(SimSpaceAngularAlpha=0,MaxAngularVelocity=10000,MaxAngularAcceleration=10000,ExternalAngularVelocity=(X=0,Y=0,Z=0))
    ExternalForce="0,0,0"
   ) -> (Pose) {
    angularDampingOverride=0.9999
    angularSpringConstant=5000
    bOverrideAngularDamping=true
    bUseGravityOverride=true
    simulationSpace=RootRelative
    alphaBoolBlend {blendOption=Linear}
    boundBone {boneName=dyn_knots_A_01}
    componentAppliedLinearAccClamp=<3000 3000 3000>
    componentLinearAccScale=<0.45 0.45 0.45>
    componentLinearVelScale=<0.45 0.45 0.45>
    gravityOverride {z=-980}
    physicsBodyDefinitions [{
      boundBone {boneName=dyn_knots_A_01}
      boxExtents=<15 20 20>
      constraintSetup {angularLimitsMax=<0 15 10>, angularLimitsMin=<0 -30 -10>, angularTarget=<1 0 0>}
      localJointOffset=<5 0 0>
     }]
    relativeSpaceBone {boneName=head}
   }
   n13 = Anim Dynamics - Bone: dyn_knots_B_01!(
    ComponentPose=@n12
    Alpha=1
    SimSpaceSettings:Anim Phys Sim Space Settings Structure=(SimSpaceAngularAlpha=0,MaxAngularVelocity=10000,MaxAngularAcceleration=10000,ExternalAngularVelocity=(X=0,Y=0,Z=0))
    ExternalForce="0,0,0"
   ) -> (Pose) {
    angularDampingOverride=0.9999
    angularSpringConstant=5000
    bOverrideAngularDamping=true
    bUseGravityOverride=true
    simulationSpace=RootRelative
    alphaBoolBlend {blendOption=Linear}
    boundBone {boneName=dyn_knots_B_01}
    componentAppliedLinearAccClamp=<3000 3000 3000>
    componentLinearAccScale=<0.45 0.45 0.45>
    componentLinearVelScale=<0.45 0.45 0.45>
    gravityOverride {z=-980}
    physicsBodyDefinitions [{
      boundBone {boneName=dyn_knots_B_01}
      boxExtents=<15 20 20>
      constraintSetup {angularLimitsMax=<0 15 10>, angularLimitsMin=<0 -30 -10>, angularTarget=<1 0 0>}
      localJointOffset=<5 0 0>
     }]
    relativeSpaceBone {boneName=head}
   }
   n14 = Anim Dynamics - Bone: dyn_knotsRope_A_01!(
    ComponentPose=@n13
    Alpha=1
    SimSpaceSettings:Anim Phys Sim Space Settings Structure=(SimSpaceAngularAlpha=0,MaxAngularVelocity=10000,MaxAngularAcceleration=10000,ExternalAngularVelocity=(X=0,Y=0,Z=0))
    ExternalForce="0,0,0"
   ) -> (Pose) {
    angularDampingOverride=0.9999
    angularSpringConstant=5000
    bOverrideAngularDamping=true
    bUseGravityOverride=true
    simulationSpace=RootRelative
    alphaBoolBlend {blendOption=Linear}
    boundBone {boneName=dyn_knotsRope_A_01}
    componentAppliedLinearAccClamp=<3000 3000 3000>
    componentLinearAccScale=<0.25 0.325 0.325>
    componentLinearVelScale=<0.45 0.55 0.55>
    gravityOverride {z=-980}
    physicsBodyDefinitions [{
      boundBone {boneName=dyn_knotsRope_A_01}
      boxExtents=<12.5 20 20>
      constraintSetup {angularLimitsMax=<0 2.5 20>, angularLimitsMin=<0 -45 -20>, angularTarget=<1 0 0>}
      localJointOffset=<6 1.25 0>
     }]
    relativeSpaceBone {boneName=head}
   }
   n15 = Anim Dynamics - Bone: dyn_knotsRope_A_02!(
    ComponentPose=@n14
    Alpha=1
    SimSpaceSettings:Anim Phys Sim Space Settings Structure=(SimSpaceAngularAlpha=0,MaxAngularVelocity=10000,MaxAngularAcceleration=10000,ExternalAngularVelocity=(X=0,Y=0,Z=0))
    ExternalForce="0,0,0"
   ) -> (Pose) {
    angularDampingOverride=0.9999
    angularSpringConstant=5000
    bOverrideAngularDamping=true
    bUseGravityOverride=true
    simulationSpace=RootRelative
    alphaBoolBlend {blendOption=Linear}
    boundBone {boneName=dyn_knotsRope_A_02}
    componentAppliedLinearAccClamp=<3000 3000 3000>
    componentLinearAccScale=<0.25 0.325 0.325>
    componentLinearVelScale=<0.45 0.55 0.55>
    gravityOverride {z=-980}
    physicsBodyDefinitions [{
      boundBone {boneName=dyn_knotsRope_A_02}
      boxExtents=<12.5 17.5 17.5>
      constraintSetup {angularLimitsMax=<0 5 25>, angularLimitsMin=<0 -45 -25>, angularTarget=<1 0 0>}
      localJointOffset=<8.5 1.25 0>
     }]
    relativeSpaceBone {boneName=head}
   }
   n16 = Anim Dynamics - Bone: dyn_knotsRope_B_01!(
    ComponentPose=@n15
    Alpha=1
    SimSpaceSettings:Anim Phys Sim Space Settings Structure=(SimSpaceAngularAlpha=0,MaxAngularVelocity=10000,MaxAngularAcceleration=10000,ExternalAngularVelocity=(X=0,Y=0,Z=0))
    ExternalForce="0,0,0"
   ) -> (Pose) {
    angularDampingOverride=0.9999
    angularSpringConstant=5000
    bOverrideAngularDamping=true
    bUseGravityOverride=true
    simulationSpace=RootRelative
    alphaBoolBlend {blendOption=Linear}
    boundBone {boneName=dyn_knotsRope_B_01}
    componentAppliedLinearAccClamp=<3000 3000 3000>
    componentLinearAccScale=<0.25 0.35 0.35>
    componentLinearVelScale=<0.45 0.6 0.6>
    gravityOverride {z=-800}
    physicsBodyDefinitions [{
      boundBone {boneName=dyn_knotsRope_B_01}
      boxExtents=<15 20 20>
      constraintSetup {angularLimitsMax=<0 2.5 20>, angularLimitsMin=<0 -45 -20>, angularTarget=<1 0 0>}
      localJointOffset=<7.5 0 0>
     }]
    relativeSpaceBone {boneName=head}
   }
   n17 = Anim Dynamics - Bone: dyn_knotsRope_B_02!(
    ComponentPose=@n16
    Alpha=1
    SimSpaceSettings:Anim Phys Sim Space Settings Structure=(SimSpaceAngularAlpha=0,MaxAngularVelocity=10000,MaxAngularAcceleration=10000,ExternalAngularVelocity=(X=0,Y=0,Z=0))
    ExternalForce="0,0,0"
   ) -> (Pose) {
    angularDampingOverride=0.9999
    angularSpringConstant=5000
    bOverrideAngularDamping=true
    bUseGravityOverride=true
    simulationSpace=RootRelative
    alphaBoolBlend {blendOption=Linear}
    boundBone {boneName=dyn_knotsRope_B_02}
    componentAppliedLinearAccClamp=<3000 3000 3000>
    componentLinearAccScale=<0.25 0.35 0.35>
    componentLinearVelScale=<0.45 0.6 0.6>
    gravityOverride {z=-800}
    physicsBodyDefinitions [{
      boundBone {boneName=dyn_knotsRope_B_02}
      boxExtents=<15 17.5 17.5>
      constraintSetup {angularLimitsMax=<0 5 25>, angularLimitsMin=<0 -45 -25>, angularTarget=<1 0 0>}
      localJointOffset=<10 0 0>
     }]
    relativeSpaceBone {boneName=head}
   }
   n18 = Anim Dynamics - Bone: deform_strapsBelt_lt_01!(
    ComponentPose=@n17
    Alpha=1
    SimSpaceSettings:Anim Phys Sim Space Settings Structure=(SimSpaceAngularAlpha=0,MaxAngularVelocity=10000,MaxAngularAcceleration=10000,ExternalAngularVelocity=(X=0,Y=0,Z=0))
    ExternalForce="0,0,0"
   ) -> (Pose) {
    angularDampingOverride=0.9999
    angularSpringConstant=2000
    bOverrideAngularDamping=true
    bUseGravityOverride=true
    simulationSpace=RootRelative
    alphaBoolBlend {blendOption=Linear}
    boundBone {boneName=deform_strapsBelt_lt_01}
    componentAppliedLinearAccClamp=<3000 3000 3000>
    componentLinearAccScale=<0.25 0.25 0.35>
    componentLinearVelScale=<0.45 0.45 0.6>
    gravityOverride {z=-980}
    physicsBodyDefinitions [{
      boundBone {boneName=deform_strapsBelt_lt_01}
      boxExtents=<20 12.5 12.5>
      constraintSetup {angularLimitsMax=<0 60 2.5>, angularLimitsMin=<0 -2.5 -2.5>, angularTarget=<1 0 0>}
      localJointOffset=<1.5 0 -0.16>
     }]
    relativeSpaceBone {boneName=head}
   }
   n19 = Anim Dynamics - Bone: deform_strapZipper_01!(
    ComponentPose=@n18
    Alpha=1
    SimSpaceSettings:Anim Phys Sim Space Settings Structure=(SimSpaceAngularAlpha=0,MaxAngularVelocity=10000,MaxAngularAcceleration=10000,ExternalAngularVelocity=(X=0,Y=0,Z=0))
    ExternalForce="0,0,0"
   ) -> (Pose) {
    angularDampingOverride=1
    angularSpringConstant=3500
    bOverrideAngularDamping=true
    bUseGravityOverride=true
    simulationSpace=RootRelative
    alphaBoolBlend {blendOption=Linear}
    boundBone {boneName=deform_strapZipper_01}
    componentAppliedLinearAccClamp=<3000 3000 3000>
    componentLinearAccScale=<0.25 0.25 0.35>
    componentLinearVelScale=<0.45 0.45 0.6>
    gravityOverride {z=-980}
    physicsBodyDefinitions [{
      boundBone {boneName=deform_strapZipper_01}
      boxExtents=<20 15 15>
      constraintSetup {angularLimitsMax=<0 2 45>, angularLimitsMin=<0 -78 -45>, angularTarget=<1 0 0>}
      localJointOffset=<5 0 0>
     }]
    relativeSpaceBone {boneName=head}
   }
   Output Pose(Result=Component To Local(ComponentPose=@n19)) {layerGroup=None}
  ]}
 EventGraph {nodes [
   Event BlueprintUpdateAnimation()
   > n1 = Sequence()
   > Branch(Condition=CheckForTail!(self:Custom Character Part Anim Instance Object Reference) -> (bHasTail:bool))
   > n2 = Cast To PlayerPawn_Athena(Object=TryGetPawnOwner())
   > Set TailMesh(TailMesh:Skeletal Mesh Component Object Reference=GetSkeletalMeshForPartType(self=@n2, PartType=MiscOrTail)) -> (Output_Get:Skeletal Mesh Component Object Reference)
   > n3 = Branch(Condition=CheckForBackbling().bIsBackblingACape)
   > SetVisibility!(
    execute
    self:Scene Component Object Reference=Get TailMesh() -> (TailMesh:Skeletal Mesh Component Object Reference)
    bNewVisibility:bool=false
    bPropagateToChildren:bool=false
   ) -> (then)
   SetVisibility(execute=@n3.else, self=Get TailMesh(), bNewVisibility=true, bPropagateToChildren=false)
   n4 = Branch(execute=@n1.then_1, Condition=Get bIsInFrontEnd())
   > Set Sim Space Settings(Sim Space Settings:Anim Phys Sim Space Settings Structure=Get Sim Space Settings_FE() -> (Sim Space Settings_FE:Anim Phys Sim Space Settings Structure)) -> (Output_Get:Anim Phys Sim Space Settings Structure)
   > Set Sim Space Settings_Jacket(Sim Space Settings_Jacket:Anim Phys Sim Space Settings Structure=Get Sim Space Settings_Jacket_FE() -> (Sim Space Settings_Jacket_FE:Anim Phys Sim Space Settings Structure)) -> (Output_Get:Anim Phys Sim Space Settings Structure)
   Set Sim Space Settings(execute=@n4.else, Sim Space Settings=Get Sim Space Settings_PIE() -> (Sim Space Settings_PIE:Anim Phys Sim Space Settings Structure))
   > Set Sim Space Settings_Jacket(Sim Space Settings_Jacket=Get Sim Space Settings_Jacket_PIE() -> (Sim Space Settings_Jacket_PIE:Anim Phys Sim Space Settings Structure))
   Event OnCharacterCustomizationLoaded()
   > n5 = Cast To PlayerPawn_Athena(Object=TryGetPawnOwner())
   > Set HeadMesh(HeadMesh:Skeletal Mesh Component Object Reference=GetSkeletalMeshForPartType(self=@n5, PartType=Head)) -> (Output_Get:Skeletal Mesh Component Object Reference)
  ]}
]
variables [
 HeadMesh {type=SkeletalMeshComponent}
 Sim Space Settings {
  type=AnimPhysSimSpaceSettings
  default {maxAngularAcceleration=10000, maxAngularVelocity=10000, simSpaceAngularAlpha=0, externalAngularVelocity=<0 0 0>}
 }
 Sim Space Settings_FE {
  type=AnimPhysSimSpaceSettings
  default {maxAngularAcceleration=48, maxAngularVelocity=6, simSpaceAngularAlpha=1, externalAngularVelocity=<0 0 0>}
 }
 Sim Space Settings_Jacket {
  type=AnimPhysSimSpaceSettings
  default {maxAngularAcceleration=10000, maxAngularVelocity=10000, simSpaceAngularAlpha=0, externalAngularVelocity=<0 0 0>}
 }
 Sim Space Settings_Jacket_FE {
  type=AnimPhysSimSpaceSettings
  default {maxAngularAcceleration=25, maxAngularVelocity=4, simSpaceAngularAlpha=1, externalAngularVelocity=<0 0 0>}
 }
 Sim Space Settings_Jacket_PIE {
  type=AnimPhysSimSpaceSettings
  default {maxAngularAcceleration=10000, maxAngularVelocity=10000, simSpaceAngularAlpha=0, externalAngularVelocity=<0 0 0>}
 }
 Sim Space Settings_PIE {
  type=AnimPhysSimSpaceSettings
  default {maxAngularAcceleration=10000, maxAngularVelocity=10000, simSpaceAngularAlpha=0, externalAngularVelocity=<0 0 0>}
 }
 TailMesh {type=SkeletalMeshComponent}
]
