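# Control Rig for the RoseScour female character's face accessory mesh that corrects hair bones as the head turns.
# Branches on JacketEquipped and isSkyDiveFloating pick one of four sets of hair rotation, offset and scale corrections.
# Head-to-spine_05 rotation, converted to Euler and remapped, weights each correction, with some multiplying two axes.
# With the jacket equipped, parent constraints also blend both hairF_G base bones between spine_05 and head.
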ControlRigBlueprint /BRCosmetics/Characters/Player/Female/Medium/Bodies/F_MED_RoseScour/Meshes/Parts/F_MED_RoseScour_FaceAcc_CtrlRig
graphs [
 RigVMFunctionLibrary {links [], nodes []}
 RigVMModel {nodes [
   RigUnit_BeginExecution = &RigUnit_BeginExecution()
   RigVMFunction_ControlFlowBranch = &RigVMFunction_ControlFlowBranch(ExecuteContext=@RigUnit_BeginExecution.ExecutePin, Condition=RigVMVariableNode() {variable=JacketEquipped}.Value)
   n1 = &RigVMFunction_ControlFlowBranch(ExecuteContext=@RigVMFunction_ControlFlowBranch.False, Condition=RigVMVariableNode() {variable=isSkyDiveFloating}.Value)
   n2 = &RigVMFunction_MathQuaternionToEuler(
    RotationOrder=XZY
    Value=&RigUnit_GetRelativeTransformForItem(Child=(Type=Bone,Name="head"), Parent=(Type=Bone,Name="spine_05"), bChildInitial=False, bParentInitial=False).RelativeTransform.Rotation
   )
   &RigUnit_ModifyTransforms(
    ItemToModify=((Item=(Type=Bone,Name="dyn_hairFBase_01"),Transform=(Rotation=(X=-0.1677,Y=-0.2549,Z=-0.0449,W=0.9513),Translation=(X=0,Y=0,Z=0),Scale3D=(X=0.8,Y=1,Z=1))))
    WeightMinimum=0
    Weight=&RigVMFunction_MathDoubleRemap(SourceMaximum=-60, SourceMinimum=-20, TargetMaximum=1, TargetMinimum=0, Value=@n2.Result.Z).Result
    ExecutePin=@n1.False
   )
   > &RigUnit_ModifyTransforms(
    ItemToModify=((Item=(Type=Bone,Name="dyn_hairF_r_rotate"),Transform=(Rotation=(X=0.0151,Y=-0.0858,Z=0.173,W=0.9811),Translation=(X=0,Y=0,Z=0),Scale3D=(X=0.9,Y=1,Z=1))))
    WeightMinimum=0
    Weight=&RigVMFunction_MathDoubleMul(
     A=&RigVMFunction_MathDoubleRemap(SourceMaximum=-60, SourceMinimum=0, TargetMaximum=1, TargetMinimum=0, Value=@n2.Result.X).Result
     B=&RigVMFunction_MathDoubleRemap(SourceMaximum=-25, SourceMinimum=-10, TargetMaximum=1, TargetMinimum=0, Value=@n2.Result.Z).Result
    ).Result
   )
   n3 = &RigVMFunction_MathQuaternionToEuler(
    RotationOrder=XZY
    Value=&RigUnit_GetRelativeTransformForItem(Child=(Type=Bone,Name="head"), Parent=(Type=Bone,Name="spine_05"), bChildInitial=False, bParentInitial=False).RelativeTransform.Rotation
   )
   &RigUnit_ModifyTransforms(
    ItemToModify=((Item=(Type=Bone,Name="dyn_hairFBase_01"),Transform=(Rotation=(X=0,Y=-0.2588,Z=0,W=0.9659),Translation=(X=0,Y=0,Z=0),Scale3D=(X=0.9,Y=1,Z=0.75))))
    WeightMinimum=0
    Weight=&RigVMFunction_MathDoubleRemap(SourceMaximum=-35, SourceMinimum=0, TargetMaximum=1, TargetMinimum=0, Value=@n3.Result.Z).Result
    ExecutePin=@n1.True
   )
   > &RigUnit_ModifyTransforms(
    ItemToModify="((Item=(Type=Bone,Name=\"dyn_hairF_l_rotate\"),Transform=(Rotation=(X=-0.0133,Y=-0.1515,Z=-0.0861,W=0.9846),Translation=(X=0,Y=0,Z=0),Scale3D=(X=1,Y=1,Z=1))),(Item=(Type=Bone,Name=\"dyn_hairF_r_rotate\"),Transform=(Rotation=(X=0.0851,Y=-0.0472,Z=-0.0396,W=0.99 ...(truncated)"
    WeightMinimum=0
    Weight=&RigVMFunction_MathDoubleRemap(SourceMaximum=30, SourceMinimum=0, TargetMaximum=1, TargetMinimum=0, Value=@n3.Result.X).Result
   )
   > &RigUnit_ModifyTransforms(
    ItemToModify="((Item=(Type=Bone,Name=\"dyn_hairF_r_rotate\"),Transform=(Rotation=(X=0.0076,Y=-0.0868,Z=0.0868,W=0.9924),Translation=(X=0,Y=0,Z=0),Scale3D=(X=1,Y=1,Z=1))),(Item=(Type=Bone,Name=\"dyn_hairF_l_rotate\"),Transform=(Rotation=(X=0.001,Y=-0.0218,Z=0.0436,W=0.9988), ...(truncated)"
    WeightMinimum=0
    Weight=&RigVMFunction_MathDoubleRemap(SourceMaximum=-30, SourceMinimum=0, TargetMaximum=1, TargetMinimum=0, Value=@n3.Result.X).Result
   )
   n4 = &RigVMFunction_ControlFlowBranch(ExecuteContext=@RigVMFunction_ControlFlowBranch.True, Condition=RigVMVariableNode() {variable=isSkyDiveFloating}.Value)
   &RigUnit_ParentConstraint(
    AdvancedSettings=(InterpolationType=Average,RotationOrderForFilter=XZY)
    Child=(Type=Bone,Name="dyn_hairF_G_r_base")
    Parents=((Item=(Type=Bone,Name="spine_05"),Weight=0.2),(Item=(Type=Bone,Name="head"),Weight=0.8))
    Weight=1
    ExecutePin=@n4.Completed
   )
   > &RigUnit_ParentConstraint(
    AdvancedSettings=(InterpolationType=Average,RotationOrderForFilter=XZY)
    Child=(Type=Bone,Name="dyn_hairF_G_l_base")
    Parents=((Item=(Type=Bone,Name="spine_05"),Weight=0.25),(Item=(Type=Bone,Name="head"),Weight=0.75))
    Weight=1
   )
   RigVMFunction_MathQuaternionToEuler = &RigVMFunction_MathQuaternionToEuler(
    RotationOrder=XZY
    Value=&RigUnit_GetRelativeTransformForItem(Child=(Type=Bone,Name="head"), Parent=(Type=Bone,Name="spine_05"), bChildInitial=False, bParentInitial=False).RelativeTransform.Rotation
   )
   Remap = &RigVMFunction_MathDoubleRemap(SourceMaximum=-50, SourceMinimum=-20, TargetMaximum=1, TargetMinimum=0, Value=@RigVMFunction_MathQuaternionToEuler.Result.Z)
   &RigUnit_ModifyTransforms(
    ItemToModify=((Item=(Type=Bone,Name="dyn_hairFBase_01"),Transform=(Rotation=(X=-0.2241,Y=-0.483,Z=-0.1294,W=0.8365),Translation=(X=0,Y=0,Z=0),Scale3D=(X=0.95,Y=1,Z=0.9))))
    WeightMinimum=0
    Weight=@Remap.Result
    ExecutePin=@n4.False
   )
   > &RigUnit_ModifyTransforms(
    ItemToModify=((Item=(Type=Bone,Name="dyn_hairE_r_base"),Transform=(Rotation=(X=0,Y=-0.4226,Z=0,W=0.9063),Translation=(X=0,Y=0,Z=0),Scale3D=(X=1,Y=1,Z=1))))
    WeightMinimum=0
    Weight=@Remap.Result
   )
   > &RigUnit_ModifyTransforms(
    ItemToModify=((Item=(Type=Bone,Name="dyn_hairF_r_rotate"),Transform=(Rotation=(X=-0.0614,Y=-0.1063,Z=0.2495,W=0.9606),Translation=(X=-0.5,Y=0,Z=0),Scale3D=(X=0.9,Y=0.95,Z=1))))
    WeightMinimum=0
    Weight=&RigVMFunction_MathDoubleMul(
     A=&RigVMFunction_MathDoubleRemap(SourceMaximum=-60, SourceMinimum=0, TargetMaximum=1, TargetMinimum=0, Value=@RigVMFunction_MathQuaternionToEuler.Result.X).Result
     B=&RigVMFunction_MathDoubleRemap(SourceMaximum=-25, SourceMinimum=-10, TargetMaximum=1, TargetMinimum=0, Value=@RigVMFunction_MathQuaternionToEuler.Result.Z).Result
    ).Result
   )
   n5 = &RigVMFunction_MathQuaternionToEuler(
    RotationOrder=XZY
    Value=&RigUnit_GetRelativeTransformForItem(Child=(Type=Bone,Name="head"), Parent=(Type=Bone,Name="spine_05"), bChildInitial=False, bParentInitial=False).RelativeTransform.Rotation
   )
   &RigUnit_ModifyTransforms(
    ItemToModify=((Item=(Type=Bone,Name="dyn_hairFBase_01"),Transform=(Rotation=(X=0,Y=-0.342,Z=0,W=0.9397),Translation=(X=0,Y=0,Z=0),Scale3D=(X=0.9,Y=1,Z=0.75))))
    WeightMinimum=0
    Weight=&RigVMFunction_MathDoubleRemap(SourceMaximum=-35, SourceMinimum=0, TargetMaximum=1, TargetMinimum=0, Value=@n5.Result.Z).Result
    ExecutePin=@n4.True
   )
   > &RigUnit_ModifyTransforms(
    ItemToModify="((Item=(Type=Bone,Name=\"dyn_hairF_l_rotate\"),Transform=(Rotation=(X=0.1819,Y=-0.1709,Z=-0.1819,W=0.9511),Translation=(X=0,Y=-0.5,Z=0.5),Scale3D=(X=1,Y=1,Z=1))),(Item=(Type=Bone,Name=\"dyn_hairF_r_rotate\"),Transform=(Rotation=(X=0.0717,Y=-0.1387,Z=-0.0962,W= ...(truncated)"
    WeightMinimum=0
    Weight=&RigVMFunction_MathDoubleRemap(SourceMaximum=20, SourceMinimum=0, TargetMaximum=1, TargetMinimum=0, Value=@n5.Result.Y).Result
   )
   > &RigUnit_ModifyTransforms(
    ItemToModify="((Item=(Type=Bone,Name=\"dyn_hairF_r_rotate\"),Transform=(Rotation=(X=-0.274,Y=-0.299,Z=-0.0507,W=0.9127),Translation=(X=0,Y=0.5,Z=0.5),Scale3D=(X=1,Y=1,Z=1))),(Item=(Type=Bone,Name=\"dyn_hairF_l_rotate\"),Transform=(Rotation=(X=0.0038,Y=-0.1736,Z=0.0215,W=0.9 ...(truncated)"
    WeightMinimum=0
    Weight=&RigVMFunction_MathDoubleRemap(SourceMaximum=-20, SourceMinimum=0, TargetMaximum=1, TargetMinimum=0, Value=@n5.Result.Y).Result
   )
  ]}
]
hierarchy [
 root {loc=<0 0 0>, rot=<0 0 0 1>}
 pelvis {parent=root, loc=<0 -0.5301 98.0571>, rot=<0.0127 0.707 -0.0127 -0.707>}
 spine_01 {parent=pelvis, loc=<6.1208 -0.9328 0>, rot=<0 0 0.034 -0.9994>}
 spine_02 {parent=spine_01, loc=<6.5997 -0.2095 0>, rot=<0 0 0.0238 -0.9997>}
 spine_03 {parent=spine_02, loc=<6.4914 -0.1237 0>, rot=<0 0 0.0149 -0.9999>}
 spine_04 {parent=spine_03, loc=<6.9598 0.0794 0>, rot=<0 0 -0.0132 -0.9999>}
 spine_05 {parent=spine_04, loc=<6.9987 0 0>, rot=<0 0 -0.0015 -1>}
 neck_01 {parent=spine_05, loc=<11.3138 0.1529 0>, rot=<0 0 -0.2003 -0.9797>}
 neck_02 {parent=neck_01, loc=<4.5627 0 0>, rot=<0 0 0.0504 -0.9987>}
 head {parent=neck_02, loc=<5.3914 0 0>, rot=<0 0 0.115 -0.9934>}
 FACIAL_C_FacialRoot {parent=head, loc=<2.1524 -0.746 -0.0002>, rot=<0.4879 -0.5118 -0.4879 -0.5118>}
 FACIAL_L_NasolabialBulge {parent=FACIAL_C_FacialRoot, loc=<3.1013 4.8026 6.2299>, rot=<0.0519 -0.3775 0.0214 -0.9243>}
 FACIAL_R_NasolabialBulge {parent=FACIAL_C_FacialRoot, loc=<-3.0992 4.8017 6.2318>, rot=<0.0519 0.3775 -0.0213 -0.9243>}
 FACIAL_L_CheekLower {parent=FACIAL_C_FacialRoot, loc=<3.4202 4.6229 5.596>, rot=<0.0287 -0.4273 0.055 -0.902>}
 FACIAL_R_CheekLower {parent=FACIAL_C_FacialRoot, loc=<-3.4174 4.6222 5.5981>, rot=<0.0287 0.4272 -0.0549 -0.902>}
 FACIAL_C_MouthUpper {parent=FACIAL_C_FacialRoot, loc=<0.0007 4.9989 7.1849>, rot=<-0.0065 0 0 -1>}
 FACIAL_L_LipUpperOuter {parent=FACIAL_C_MouthUpper, loc=<1.6456 0.2182 1.302>, rot=<-0.0984 -0.2293 -0.0983 -0.9634>}
 FACIAL_R_LipUpperOuter {parent=FACIAL_C_MouthUpper, loc=<-1.6469 0.219 1.3023>, rot=<-0.0984 0.2293 0.0983 -0.9634>}
 FACIAL_L_LipCorner {parent=FACIAL_C_MouthUpper, loc=<2.132 0.5273 0.5574>, rot=<0 -0.2078 0 -0.9782>}
 FACIAL_R_LipCorner {parent=FACIAL_C_MouthUpper, loc=<-2.1325 0.527 0.5573>, rot=<0 0.2078 0 -0.9782>}
 FACIAL_C_LowerLipRotation {parent=FACIAL_C_FacialRoot, loc=<-0.0006 2.6342 2.1678>, rot=<0.3359 0 0 -0.9419>}
 FACIAL_C_MouthLower {parent=FACIAL_C_LowerLipRotation, loc=<0.0012 -1.3438 5.3811>, rot=<-0.342 0 0 -0.9397>}
 FACIAL_L_LipLowerOuter {parent=FACIAL_C_MouthLower, loc=<1.5068 0.939 1.0182>, rot=<0.0913 -0.2321 0.0318 -0.9679>}
 FACIAL_R_LipLowerOuter {parent=FACIAL_C_MouthLower, loc=<-1.51 0.9394 1.0188>, rot=<0.0913 0.2321 -0.0318 -0.9679>}
 FACIAL_C_Jaw {parent=FACIAL_C_FacialRoot, loc=<-0.0006 2.6342 2.1678>, rot=<0.3359 0 0 -0.9419>}
 FACIAL_L_ChinSide {parent=FACIAL_C_Jaw, loc=<2.4587 0.2989 6.8168>, rot=<-0.2208 -0.3738 0.0499 -0.8995>}
 FACIAL_R_ChinSide {parent=FACIAL_C_Jaw, loc=<-2.4562 0.2994 6.8156>, rot=<-0.2208 0.3738 -0.0499 -0.8995>}
 dyn_frontHair_base {parent=head, loc=<10.719 8.2676 0>, rot=<0.2025 0.6775 0.6775 0.2025>}
 dyn_frontHair_01 {parent=dyn_frontHair_base, loc=<0 0 0>, rot=<0 0 0 1>}
 dyn_frontHair_02 {parent=dyn_frontHair_01, loc=<5.0006 0 0>, rot=<0 -0.457 0 -0.8895>}
 dyn_hairA_l_base {parent=head, loc=<5.7026 2.8711 -6.6337>, rot=<-0.0029 0.7071 0.7071 -0.0029>}
 dyn_hairA_l_01 {parent=dyn_hairA_l_base, loc=<0 0 0>, rot=<0 0 0 1>}
 dyn_hairA_l_02 {parent=dyn_hairA_l_01, loc=<6 0 0>, rot=<0 0 0 1>}
 dyn_hairB_l_base {parent=head, loc=<3.347 -1.1195 -7.6202>, rot=<0.0039 -0.939 -0.0014 -0.3439>}
 dyn_hairB_l_01 {parent=dyn_hairB_l_base, loc=<0 0 0>, rot=<0 0 0 1>}
 dyn_hairC_l_base {parent=head, loc=<-0.1103 -2.8016 -9.2202>, rot=<0.0038 -0.9131 -0.0017 -0.4077>}
 dyn_hairC_l_01 {parent=dyn_hairC_l_base, loc=<0 0 0>, rot=<0 0 0 1>}
 dyn_hairD_l_base {parent=head, loc=<1.3727 -3.7466 -9.573>, rot=<0.2601 -0.7692 0.3068 -0.4966>}
 dyn_hairD_l_01 {parent=dyn_hairD_l_base, loc=<0 0 0>, rot=<0 0 0 1>}
 dyn_hairE_l_base {parent=head, loc=<-0.5737 -6.5593 -8.1999>, rot=<0.1786 -0.8489 0.3144 -0.3856>}
 dyn_hairE_l_01 {parent=dyn_hairE_l_base, loc=<0 0 0>, rot=<0 0 0 1>}
 dyn_hairA_r_base {parent=head, loc=<5.7026 2.8711 6.634>, rot=<-0.0029 0.7071 0.7071 -0.0029>}
 dyn_hairA_r_01 {parent=dyn_hairA_r_base, loc=<0 0 0>, rot=<0 0 0 1>}
 dyn_hairA_r_02 {parent=dyn_hairA_r_01, loc=<6 0 0>, rot=<0 0 0 1>}
 dyn_hairB_r_base {parent=head, loc=<1.8824 -0.7207 8.3739>, rot=<0.4666 -0.1064 0.8671 -0.1379>}
 dyn_hairB_r_01 {parent=dyn_hairB_r_base, loc=<0 0 0>, rot=<0 0 0 1>}
 dyn_hairC_r_base {parent=head, loc=<2.7197 -2.7322 8.9702>, rot=<0.5314 -0.052 0.845 -0.0298>}
 dyn_hairC_r_01 {parent=dyn_hairC_r_base, loc=<0 0 0>, rot=<0 0 0 1>}
 dyn_hairD_r_base {parent=head, loc=<0.7006 -4.704 8.9249>, rot=<0.678 -0.1942 0.6901 -0.1622>}
 dyn_hairD_r_01 {parent=dyn_hairD_r_base, loc=<0 0 0>, rot=<0 0 0 1>}
 dyn_hairE_r_base {parent=head, loc=<-1.0201 -6.8865 7.1683>, rot=<0.5423 -0.2389 0.7903 -0.1558>}
 dyn_hairE_r_01 {parent=dyn_hairE_r_base, loc=<0 0 0>, rot=<0 0 0 1>}
 dyn_hairF_G_l_base {parent=head, loc=<-1.2825 -3.3171 -6.51>, rot=<0 -1 0 0>}
 dyn_hairG_l_01 {parent=dyn_hairF_G_l_base, loc=<0 0 0>, rot=<0 0 0 1>}
 dyn_hairF_G_r_base {parent=head, loc=<-1.2825 -3.3171 6.5096>, rot=<0 0 1 0>}
 dyn_hairG_r_01 {parent=dyn_hairF_G_r_base, loc=<0 0 0>, rot=<0 0 0 1>}
 dyn_hairFBase_base {parent=head, loc=<4.4459 -7.9872 0>, rot=<0.0029 -0.7071 0.7071 -0.0029>}
 dyn_hairFBase_01 {parent=dyn_hairFBase_base, loc=<0 0 0>, rot=<0 0 0 1>}
 dyn_hairF_r_rotate {parent=dyn_hairFBase_01, loc=<0 -1.5 0>, rot=<0 0 0 1>}
 dyn_hairF_r_base {parent=dyn_hairF_r_rotate, loc=<0.0257 -1.5179 0.4202>, rot=<-0.2159 0.0668 -0.0148 -0.974>}
 dyn_hairF_r_01 {parent=dyn_hairF_r_base, loc=<0 0 0>, rot=<0 0 0 1>}
 dyn_hairF_l_rotate {parent=dyn_hairFBase_01, loc=<0 1.5 0>, rot=<0 0 0 1>}
 dyn_hairF_l_base {parent=dyn_hairF_l_rotate, loc=<0.0257 1.518 0.4202>, rot=<0.2159 0.0668 0.0148 -0.974>}
 dyn_hairF_l_01 {parent=dyn_hairF_l_base, loc=<0 0 0>, rot=<0 0 0 1>}
 FX_Head_Horn_L {parent=head, type=Null, loc=<15.68 9.2514 -7.9752>, rot=<0 0 0 1>}
 FX_Head_Horn_R {parent=head, type=Null, loc=<15.68 9.2514 8.1192>, rot=<0 0 0 1>}
 clavicle_l {parent=spine_05, loc=<5.4831 -0.1053 -3.2946>, rot=<0.0465 -0.7289 -0.009 -0.6829>}
 clavicle_r {parent=spine_05, loc=<5.4831 -0.1053 3.2946>, rot=<0.7289 0.0465 -0.6829 0.009>}
 neck_01_ctrl {controlType=EULER_TRANSFORM, type=Control, loc=<0 0 0>, rot=<0 0 0 1>}
 neck_02_ctrl {controlType=EULER_TRANSFORM, parent=neck_01_ctrl, type=Control, loc=<0 0 0>, rot=<0 0 0 1>}
 head_ctrl {controlType=EULER_TRANSFORM, parent=neck_02_ctrl, type=Control, loc=<0 0 0>, rot=<0 0 0 1>}
]
variables [
 JacketEquipped {type=bool, default=False}
 isSkyDiveFloating {type=bool, default=False}
]
