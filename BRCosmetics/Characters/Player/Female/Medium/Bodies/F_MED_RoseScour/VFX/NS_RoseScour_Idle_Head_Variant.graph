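# Idle VFX for the RoseScour character's head: persistent orange glows, named as decal and light, that follow the skeletal mesh.
# One emitter bursts two spawn groups of two particles, pinned by partition index to bones or sockets with per-group offsets, feeding three sprite renderers.
# Glows fade in by curve, scale with time of day and zero their alpha when User.bisHeadHidden is set.
# Dynamic material parameters receive a bone-rotated capsule glow direction.
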NiagaraSystem /BRCosmetics/Characters/Player/Female/Medium/Bodies/F_MED_RoseScour/VFX/NS_RoseScour_Idle_Head_Variant
dependencies {
 usedDataInterfaces [&NiagaraDataInterfaceCurve, &NiagaraDataInterfaceSkeletalMesh]
 usedDynamicInputs []
 usedModules [
  &self:SystemSpawnScript.SystemScriptSource.SystemScriptGraph.NiagaraNodeAssignment_1.SetVariables_#1
  &SystemState
 ]
}
emitters {GlowDecalAndLight {
  emitterSpawnScript [/SetVariables(Emitter.CapsuleGlowDirection:Vector3f=<1 0.1 0>) {isSetParameters=true}]
  emitterUpdateScript [
   &EmitterState(
    Inactive Response=::NewEnumerator1
    Life Cycle Mode=::NewEnumerator1
    Loop Behavior=::NewEnumerator1
    Loop Duration Mode=::NewEnumerator1
    UseLoopDelay=false
   )
   &SpawnBurst_Instantaneous(Spawn Count=2, Use Spawn Probability=false, UseLoopCountLimit=false) {module=SpawnBurst_Instantaneous001}
   &SpawnBurst_Instantaneous(Spawn Count=2, Spawn Group=1, Use Spawn Probability=false, UseLoopCountLimit=false)
  ]
  particleSpawnScript [
   /SetVariables(Particles.VisibilityTag=&SpawnGroup()) {isSetParameters=true}
   &PartitionParticles()
   &InitializeParticle(
    Color=<rgba 0.5 0.1906 0 0.25>
    Lifetime=0.5
    Sprite Size Mode=::NewEnumerator3
    Uniform Sprite Size=20
    UsePositionOffset=false
    Write Material Random=true
   ) {module=InitializeParticle001}
   /SetVariables(Particles.CapsuleCameraOffset=1, Particles.CapsuleSize:Vector2f=<15 20>) {isSetParameters=true}
  ]
  particleUpdateScript [
   &ParticleState(Kill Particles When Lifetime Has Elapsed=false)
   &SkeletalMeshLocation(
    Bone / Socket Index=$Particles.PartitionParticles.Partition
    MaskSpawnGroup=false
    Mesh Sampling Type=::NewEnumerator5
    Sampled Position Offset=&MergeHelper_VectorSwitch(
     0=<0 1 0>
     1=&MergeHelper_VectorSwitch(0=<5 -5 5>, 1=<5 -5 -5>, 2=?, 3=?, Index=$Particles.PartitionParticles.Partition, Num=::NewEnumerator0)
     2=?
     3=?
     Index=&SpawnGroup()
     Num=::NewEnumerator0
    )
    Skeletal Mesh=$System.SkeletalMesh
    Socket Sampling Mode=::NewEnumerator5
   )
   &ScaleSpriteSize(Uniform Curve Sprite Scale=DataInterface() {propertyValues {Curve {keys [
        {arriveTangent=1.1179, interpMode=RCIM_Cubic, leaveTangent=1.1179, tangentMode=RCTM_User, time=0, value=0.6}
        {arriveTangent=0.0224, interpMode=RCIM_Cubic, leaveTangent=0.0224, tangentMode=RCTM_User, time=1, value=1}
       ]}}})
   &CameraOffset()
   &ScaleColor(
    Scale Alpha=&FloatFromCurve(
     FloatCurve=DataInterface() {propertyValues {Curve {keys [
         {arriveTangent=4.1785, interpMode=RCIM_Cubic, leaveTangent=4.1785, tangentMode=RCTM_User, time=0, value=0}
         {arriveTangent=1.5009, interpMode=RCIM_Cubic, leaveTangent=1.5009, tangentMode=RCTM_User, time=0.2, value=0.55}
         {arriveTangent=0.0224, interpMode=RCIM_Cubic, leaveTangent=0.0224, tangentMode=RCTM_User, time=1, value=1}
        ]}}}
     Scale Curve=&MakeCustomFloatFromBool(
      Bool=&Boolean_NotOperation(Bool=$User.bisHeadHidden)
      True Float=&FNTimeOfDayScalar(Night Scale=0.75, Morning Scale=1, Evening Scale=1)
      False Float=0
     )
    )
    Scale RGB=&VectorFromFloat(Value=&FNTimeOfDayScalar(
      Night Scale=&MergeHelper_FloatSwitch(0=0.25, 1=1.1, 2=?, 3=?, Index=&SpawnGroup(), Num=::NewEnumerator0)
      Morning Scale=1
      Evening Scale=1
     ))
    ScaleA=true
    ScaleRGB=true
   )
   /SetVariables(Particles.CapsuleGlowColor:LinearColor=&MakeLinearColorFromVectorAndFloat(
     "Vector (RGB)"=&MakeVectorFromLinearColorRGB(Color=<rgba 0.5 0.2083 0 1>)
     "Float (Alpha)"=&FloatFromCurve(
      FloatCurve=DataInterface() {propertyValues {Curve {keys [
          {arriveTangent=4.1785, interpMode=RCIM_Cubic, leaveTangent=4.1785, tangentMode=RCTM_User, time=0, value=0}
          {arriveTangent=1.5009, interpMode=RCIM_Cubic, leaveTangent=1.5009, tangentMode=RCTM_User, time=0.2, value=0.55}
          {arriveTangent=0.0224, interpMode=RCIM_Cubic, leaveTangent=0.0224, tangentMode=RCTM_User, time=1, value=1}
         ]}}}
      Scale Curve=&MakeCustomFloatFromBool(
       Bool=&Boolean_NotOperation(Bool=$User.bisHeadHidden)
       True Float=&FNTimeOfDayScalar(Night Scale=0.35, Morning Scale=1, Evening Scale=1)
       False Float=0
      )
     )
    )) {isSetParameters=true}
   /SetVariables(Particles.SpriteSize_GlowTips:Vector2f=<20 20>) {isSetParameters=true}
   /SetVariables(Particles.DMP_GlowTips:Vector4f=<3.5 1 1 1>) {isSetParameters=true}
   &DynamicMaterialParameters(
    Index 0 Param 1=&MakeFloatFromVector(
     Vector=&RotateVectorByQuaternion(Vector To Rotate=$Emitter.CapsuleGlowDirection, Quaternion=$Particles.SkeletalMeshLocation.SampledRotation)
     Channel=::NewEnumerator0
    )
    Index 0 Param 2=&MakeFloatFromVector(
     Vector=&RotateVectorByQuaternion(Vector To Rotate=$Emitter.CapsuleGlowDirection, Quaternion=$Particles.SkeletalMeshLocation.SampledRotation)
     Channel=::NewEnumerator1
    )
    Index 0 Param 3=&MakeFloatFromVector(
     Vector=&RotateVectorByQuaternion(Vector To Rotate=$Emitter.CapsuleGlowDirection, Quaternion=$Particles.SkeletalMeshLocation.SampledRotation)
     Channel=::NewEnumerator2
    )
    Index 0 Param 4=20
    Module.Param0WriteEnabled=true
    Module.Param1WriteEnabled=true
    Module.Param2WriteEnabled=true
    Module.Param3WriteEnabled=true
   )
  ]
  rendererClasses [&NiagaraSpriteRendererProperties]
  renderers [
   {rendererClass=&NiagaraSpriteRendererProperties}
   {rendererClass=&NiagaraSpriteRendererProperties, rendererIndex=1}
   {rendererClass=&NiagaraSpriteRendererProperties, rendererIndex=2}
  ]
 }}
summary {
 systemName=NS_RoseScour_Idle_Head_Variant
 emitters [{emitterName=GlowDecalAndLight, rendererClasses [&NiagaraSpriteRendererProperties]}]
 userVariables [User.bisHeadHidden {type=NiagaraBool}]
}
