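# Niagara idle effect for a variant of the RoseScour character body, emitting warm-colored sprites from skeletal mesh sockets assigned by spawn group.
# LightSpores drift upward with curl noise and pulsing size, while larger PixieDust sprites inherit source movement scaled by system velocity.
# Alpha fades with distance and is masked per socket group by hidden-foot, hidden-hand and jacket user parameters.
# PixieDust opacity also varies by quality level and front-end context.
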NiagaraSystem /BRCosmetics/Characters/Player/Female/Medium/Bodies/F_MED_RoseScour/VFX/NS_RoseScour_Idle_Variant
dependencies {
 usedDataInterfaces [
  &NiagaraDataInterfaceCurve
  &NiagaraDataInterfaceVector2DCurve
  &NiagaraDataInterfaceArrayFloat
  &NiagaraDataInterfaceSkeletalMesh
 ]
 usedDynamicInputs [
  &ClampFloat
  &RemapRange
  &VectorLength
  &FNTimeOfDayScalar
 ]
 usedModules [
  &self:SystemSpawnScript.SystemScriptSource.SystemScriptGraph.NiagaraNodeAssignment_0.SetVariables_#1
  &self:SystemSpawnScript.SystemScriptSource.SystemScriptGraph.NiagaraNodeAssignment_8.SetVariables_#2
  &SystemState
  &InterpolateOverTime
  &self:SystemSpawnScript.SystemScriptSource.SystemScriptGraph.NiagaraNodeAssignment_1.SetVariables_#3
  &self:SystemSpawnScript.SystemScriptSource.SystemScriptGraph.NiagaraNodeAssignment_3.SetVariables_#4
  &BlendSpaceHelper
  &self:SystemSpawnScript.SystemScriptSource.SystemScriptGraph.NiagaraNodeAssignment_2.SetVariables_#5
 ]
}
emitters {
 LightSpores {
  emitterUpdateScript [
   &EmitterState(
    Inactive Response=::NewEnumerator1
    Life Cycle Mode=::NewEnumerator1
    Loop Behavior=::NewEnumerator1
    Loop Duration Mode=::NewEnumerator1
    UseLoopDelay=false
   )
   &SpawnRate(SpawnRate=20, Use Spawn Probability=true) {module=SpawnRate001}
  ]
  particleSpawnScript [
   &FN_SetSpawnGroup(Number of Spawn Groups:NiagaraInt32=$System.SpawnGroups)
   &InitializeParticle(
    Color=<rgba 1 0.4667 0.2 1>
    Lifetime Max=&FN_FrontEndFloatModifier(Gameplay=2.5, Front End=1)
    Lifetime Min=2
    Lifetime Mode=::NewEnumerator1
    Sprite Size Mode=::NewEnumerator1
    Uniform Sprite Size Max=5
    Uniform Sprite Size Min=1
    UsePositionOffset=false
    Write Material Random=true
   )
   &SkeletalMeshLocation(
    Bone / Socket Index=$Particles.SpawnGroup
    MaskSpawnGroup=false
    Mesh Sampling Type=::NewEnumerator5
    Skeletal Mesh=$System.SkeletalMesh
    Socket Sampling Mode=::NewEnumerator5
   ) {module=SkeletalMeshLocation003}
   &ShapeLocation(
    MaskSpawnGroup=false
    Override Seed=false
    Shape Origin=$StackContext.SkeletalMeshLocation003.SampledPosition
    Sphere Radius=10
    Sphere Surface Distribution=1
    UseOwnerScale=true
   )
  ]
  particleUpdateScript [
   &ParticleState()
   &Drag(
    Drag=&FloatFromCurve(
     FloatCurve=DataInterface() {propertyValues {Curve {keys [{interpMode=RCIM_Cubic, tangentMode=RCTM_User, time=0, value=0}, {interpMode=RCIM_Cubic, tangentMode=RCTM_User, time=1, value=1}]}}}
     Scale Curve=&RandomRangeFloat(Minimum=0, Maximum=5)
    )
    Use Linear Drag=true
    Use Rotational Drag=true
   ) {module=Drag001}
   &CameraOffset(Camera Offset Amount=&RandomRangeFloat(Minimum=15, Maximum=25))
   &AccelerationForce(Acceleration=&RandomRangeVector(Minimum=<0 0 5>, Maximum=<0 0 10>))
   &CurlNoiseForce(Bias Noise=false, Noise Frequency=2, Noise Strength=5, Pan Noise=false)
   &SolveForcesAndVelocity(Clamp Velocity=false, Limit Acceleration=false)
   &ScaleSpriteSize(
    Uniform Curve Scale=&Waveform(
     "Waveform [1]"=::NewEnumerator0
     "[1] Frequency"=&RandomRangeFloat(Minimum=0, Maximum=1.8831)
     "[1] Amplitude Min/Max"=<0.7 1>
    )
    Uniform Curve Sprite Scale=DataInterface() {propertyValues {Curve {keys [
        {interpMode=RCIM_Cubic, tangentMode=RCTM_User, time=0, value=1}
        <at 0.9 1>
        {arriveTangent=-4.3352, interpMode=RCIM_Cubic, leaveTangent=-4.3352, tangentMode=RCTM_User, time=1, value=0}
       ]}}}
   ) {module=ScaleSpriteSize001}
   &ScaleColor(
    Scale Alpha=&FloatFromCurve(
     FloatCurve=DataInterface() {propertyValues {Curve {
        defaultValue=0
        keys [
         {interpMode=RCIM_Cubic, tangentMode=RCTM_User, time=0, value=0}
         {interpMode=RCIM_Cubic, tangentMode=RCTM_User, time=0.5, value=1}
         {tangentMode=RCTM_User, time=1, value=0}
        ]
       }}}
     Scale Curve=&Multiply_Float(
      A=&FloatFromCurve(
       FloatCurve=DataInterface() {propertyValues {Curve {
          defaultValue=0
          keys [{interpMode=RCIM_Cubic, tangentMode=RCTM_User, time=0, value=0}, {interpMode=RCIM_Cubic, tangentMode=RCTM_User, time=1, value=1}]
         }}}
       CurveIndex=&NormalizeDistanceRange(
        Start Position:NiagaraPosition=&AddVectorToPosition(Position=$Engine.Owner.Position, Vector=<0 0 75>)
        End Position:NiagaraPosition=$Particles.Position
        Distance:NiagaraFloat=150
        Invert Normalized Range:NiagaraBool=true
       )
       Scale Curve=1.25
      )
      B=&Multiply_Float(
       A=&MergeHelper_FloatSwitch(
        0=1
        1=1
        2=1
        3=&MakeFloatFromInt(Int=&MakeCustomIntFromBool(Bool=$User.bisFootHidden, True INT=0, False INT=1))
        4=1
        5=&MakeFloatFromInt(Int=&MakeCustomIntFromBool(Bool=$User.bisFootHidden, True INT=0, False INT=1))
        6=&Lerp_Float(
         A=&MakeFloatFromInt(Int=&MakeCustomIntFromBool(Bool=$User.bisHandHidden, True INT=0, False INT=1))
         B=0
         Alpha=$User.bHasJacket
        )
        7=&Lerp_Float(A=1, B=0, Alpha=$User.bHasJacket)
        8=&Lerp_Float(
         A=&MakeFloatFromInt(Int=&MakeCustomIntFromBool(Bool=$User.bisHandHidden, True INT=0, False INT=1))
         B=0
         Alpha=$User.bHasJacket
        )
        9=&Lerp_Float(A=1, B=0, Alpha=$User.bHasJacket)
        Index=$Particles.SpawnGroup
        Num=::NewEnumerator8
       )
       B=&MakeCustomFloatFromBool(
        Bool=&SetBoolByIntComparison(A=$Particles.SpawnGroup, B=9, Comparison Type=::NewEnumerator0)
        True Float=&MakeFloatFromInt(Int=&MakeCustomIntFromBool(Bool=$User.bisFootHidden, True INT=0, False INT=1))
        False Float=1
       )
      )
     )
    )
    ScaleA=true
    ScaleRGB=false
   ) {module=ScaleColor001}
  ]
  rendererClasses [&NiagaraSpriteRendererProperties]
  renderers [
   {rendererClass=&NiagaraSpriteRendererProperties}
   {rendererClass=&NiagaraSpriteRendererProperties, rendererIndex=1}
  ]
 }
 PixieDust {
  emitterUpdateScript [
   &EmitterState(
    Inactive Response=::NewEnumerator1
    Life Cycle Mode=::NewEnumerator1
    Loop Behavior=::NewEnumerator1
    Loop Duration Mode=::NewEnumerator1
    UseLoopDelay=false
   )
   &SpawnRate(SpawnRate=30, Use Spawn Probability=false)
  ]
  particleSpawnScript [
   &FN_SetSpawnGroup(Number of Spawn Groups:NiagaraInt32=$System.SpawnGroups)
   &InitializeParticle(
    Color Maximum=<rgba 0.808 0.4909 0.1856 0.75>
    Color Minimum=<rgba 0.9445 0.5331 0.1371 0.75>
    Color Mode=::NewEnumerator2
    Lifetime Max=3
    Lifetime Min=2.5
    Lifetime Mode=::NewEnumerator1
    Sprite Rotation Mode=::NewEnumerator2
    Sprite Size Mode=::NewEnumerator1
    Uniform Sprite Size Max=30
    Uniform Sprite Size Min=15
    UsePositionOffset=false
    Write Material Random=true
   )
   &SkeletalMeshLocation(
    Bone / Socket Index=$Particles.SpawnGroup
    MaskSpawnGroup=false
    Mesh Sampling Type=::NewEnumerator5
    Skeletal Mesh=$System.SkeletalMesh
    Socket Sampling Mode=::NewEnumerator5
   ) {module=SkeletalMeshLocation001}
   &DynamicMaterialParameters(
    Index 0 Param 1=&RandomRangeFloat(Minimum=-0.025, Maximum=0.025)
    Index 0 Param 2=&RandomRangeFloat(Minimum=0, Maximum=32)
    Module.Index0Param0Write=true
    Module.Index0Param1Write=true
    Module.Index0Param2Write=false
    Module.Index0Param3Write=false
   )
   &CameraOffset(Camera Offset Amount=&MergeHelper_FloatSwitch(0=10, 1=10, 2=25, 3=25, 4=25, 5=25, 6=25, 7=25, 8=25, 9=25, Index=0, Num=::NewEnumerator8))
  ]
  particleUpdateScript [
   &ParticleState()
   &SkeletalMeshLocation(
    Bone / Socket Index=$Particles.SpawnGroup
    MaskSpawnGroup=false
    Mesh Sampling Type=::NewEnumerator5
    Orientation Sampling=::NewEnumerator1
    Position Sampling=::NewEnumerator2
    Skeletal Mesh=$System.SkeletalMesh
    Socket Sampling Mode=::NewEnumerator5
    Use Sampled Mesh Attribute as Ribbon ID=false
   )
   &ScaleSpriteSize(Uniform Curve Sprite Scale=DataInterface() {propertyValues {Curve {keys [
        {arriveTangent=1.3913, interpMode=RCIM_Cubic, leaveTangent=1.3913, tangentMode=RCTM_User, time=0, value=0.75}
        {tangentMode=RCTM_User, time=1, value=1.25}
       ]}}})
   &AccelerationForce(Acceleration=&RandomRangeVector(Minimum=<0 0 5>, Maximum=<0 0 11>))
   &ScaleColor(
    Scale Alpha=&Multiply_Float(
     A=&FloatFromCurve(
      FloatCurve=DataInterface() {propertyValues {Curve {
         defaultValue=0
         keys [
          {arriveTangent=8.9923, interpMode=RCIM_Cubic, leaveTangent=8.9923, tangentMode=RCTM_User, time=0, value=0}
          {interpMode=RCIM_Cubic, tangentMode=RCTM_User, time=0.25, value=0.8}
          {tangentMode=RCTM_User, time=1.25, value=0}
         ]
        }}}
      CurveIndex=$Particles.Age
      Scale Curve=&Multiply_Float(
       A=&FloatFromCurve(
        FloatCurve=DataInterface() {propertyValues {Curve {
           defaultValue=0
           keys [
            {interpMode=RCIM_Cubic, tangentMode=RCTM_User, time=0, value=0}
            {arriveTangent=2.3687, interpMode=RCIM_Cubic, leaveTangent=2.3687, tangentMode=RCTM_User, time=1, value=1}
           ]
          }}}
        CurveIndex=&NormalizeDistanceRange(
         Start Position:NiagaraPosition=$Particles.SkeletalMeshLocation.SampledPosition
         End Position:NiagaraPosition=$Particles.Position
         Distance:NiagaraFloat=50
         Invert Normalized Range:NiagaraBool=true
        )
       )
       B=&Multiply_Float(
        A=&MergeHelper_FloatSwitch(
         0=1
         1=1
         2=1
         3=&MakeFloatFromInt(Int=&MakeCustomIntFromBool(Bool=$User.bisFootHidden, True INT=0, False INT=1))
         4=1
         5=&MakeFloatFromInt(Int=&MakeCustomIntFromBool(Bool=$User.bisFootHidden, True INT=0, False INT=1))
         6=&Lerp_Float(
          A=&MakeFloatFromInt(Int=&MakeCustomIntFromBool(Bool=$User.bisHandHidden, True INT=0, False INT=1))
          B=0
          Alpha=$User.bHasJacket
         )
         7=&Lerp_Float(A=1, B=0, Alpha=$User.bHasJacket)
         8=&Lerp_Float(
          A=&MakeFloatFromInt(Int=&MakeCustomIntFromBool(Bool=$User.bisHandHidden, True INT=0, False INT=1))
          B=0
          Alpha=$User.bHasJacket
         )
         9=&Lerp_Float(A=1, B=0, Alpha=$User.bHasJacket)
         Index=$Particles.SpawnGroup
         Num=::NewEnumerator8
        )
        B=&MakeCustomFloatFromBool(
         Bool=&SetBoolByIntComparison(A=$Particles.SpawnGroup, B=9, Comparison Type=::NewEnumerator0)
         True Float=&MakeFloatFromInt(Int=&MakeCustomIntFromBool(Bool=$User.bisFootHidden, True INT=0, False INT=1))
         False Float=1
        )
       )
      )
     )
     B=&FN_FrontEndFloatModifier(
      Gameplay=&SelectFloatFromArray(Direct Array Index=$Engine.QualityLevel, Float Selection Array=DataInterface() {propertyValues {FloatData [0.25,1,1,1,1]}})
      Front End=&SelectFloatFromArray(Direct Array Index=$Engine.QualityLevel, Float Selection Array=DataInterface() {propertyValues {FloatData [0.5,1,1,1,1]}})
     )
    )
    Scale RGB=<1.1 1.1 1.1>
    ScaleA=true
    ScaleRGB=true
   )
   &InheritSourceMovement(
    Applied Position Scale=&VectorFromFloat(Value=&FloatFromCurve(
      FloatCurve=DataInterface() {propertyValues {Curve {
         defaultValue=0
         keys [
          {arriveTangent=3.5982, interpMode=RCIM_Cubic, leaveTangent=3.5982, tangentMode=RCTM_User, time=0, value=0.1}
          {arriveTangent=1.1179, interpMode=RCIM_Cubic, leaveTangent=1.1179, tangentMode=RCTM_User, time=0.2042, value=0.6228}
          {arriveTangent=0.0224, interpMode=RCIM_Cubic, leaveTangent=0.0224, tangentMode=RCTM_User, time=1, value=1}
         ]
        }}}
      CurveIndex=$System.NormalizedVelocityRange
     ))
    ApplyForce=false
    ApplyPosition=true
    ApplyPositionFalloff=false
   )
   &SolveForcesAndVelocity(Clamp Velocity=false, Limit Acceleration=false)
   &KillParticles(Kill Particles=&SetBoolByFloatComparison(
     A=&MakeFloatFromVector4(Vector 4=&MakeVector4_FromLinearColor(LinearColor=$Particles.Color), Channel=::NewEnumerator3)
     B=0.01
     Comparison Type=::NewEnumerator7
    ))
  ]
  rendererClasses [&NiagaraSpriteRendererProperties]
  renderers [{rendererClass=&NiagaraSpriteRendererProperties}]
 }
}
summary {
 systemName=NS_RoseScour_Idle_Variant
 emitters [
  {emitterName=LightSpores, rendererClasses [&NiagaraSpriteRendererProperties]}
  {emitterName=PixieDust, rendererClasses [&NiagaraSpriteRendererProperties]}
 ]
 userVariables [
  User.bHasJacket {type=NiagaraFloat, defaultValue=0, description=""}
  User.bisFootHidden {type=NiagaraBool}
  User.bIsFrontEnd {type=NiagaraBool}
  User.bIsFrontEndPreview {type=NiagaraBool}
  User.bisHandHidden {type=NiagaraBool}
 ]
}
